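# Material function for the Blackout2 variant of the CamoSE cosmetic wheel, outputting material attributes with a colored Fresnel glow in emissive.
# Vertex color red blends the rim and tire normal maps, which feed a classic Fresnel function borrowed from the Cola_Holographic wheel, with an adjustable exponent.
# The Fresnel result is tinted by the Custom Color parameter, boosted by a constant and written to EmissiveColor, the only attribute set here.
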MaterialFunction /VehicleCosmetics/Wheels/CamoSE/Materials/MF_Wheel_CamoSE_Blackout2
graph {nodes [
  n1 = TextureSample() {samplerSource=SSM_Wrap_WorldGroupSettings, samplerType=SAMPLERTYPE_Normal, texture=/VehicleCosmetics/Wheels/CamoSE/Textures/T_Camo_Rim_N}
  n2 = TextureSample() {samplerSource=SSM_Wrap_WorldGroupSettings, samplerType=SAMPLERTYPE_Normal, texture=/VehicleCosmetics/Wheels/CamoSE/Textures/T_Camo_Tire_N}
  n3 = VertexColor()
  n4 = Multiply(
   A=VectorParameter() {parameterName=Custom Color, defaultValue=<rgba 0 0.8796 0.1095 1>}.RGB
   B=MaterialFunctionCall!(Normal=LinearInterpolate(A=@n1.RGBA, B=@n2.RGBA, Alpha=@n3.R), Exponent=ScalarParameter() {defaultValue=2, parameterName=Fresnel EXP}) {materialFunction=/VehicleCosmetics/Wheels/Cola_Holographic/Materials/MF_Fresnel_Classic}.Result
  )
  FunctionOutput(None=SetMaterialAttributes!(MaterialAttributes, EmissiveColor=Multiply(A=@n4) {constB=1.5}) {attributeSetTypes [#1]}) {outputName=Material Attributes}
 ]}
