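# Material function for the camo netting on the CamoSE cosmetic wheel, outputting material attributes for an alpha-cut, wind-animated net.
# Custom and accent colors blend by a camo texture, get overlaid with the net mask's green channel and feed Emissive rather than Base Color.
# The mask's red channel cuts opacity, alongside a net normal map.
# Time-driven sine waves over rotated UVs, weighted by vertex alpha, displace vertices along a wind direction.
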MaterialFunction /VehicleCosmetics/Wheels/CamoSE/Materials/MF_Wheel_CamoSE_Net
graph {
 properties {previewBlendMode=BLEND_Masked}
 nodes [
  n1 = VectorParameter() {group=00_Base, parameterName=Custom Color, defaultValue=<rgba 0.0116 0.552 0.1912 1>}
  n2 = VectorParameter() {group=00_Base, parameterName=Accent Color, defaultValue=<rgba 0.1912 0.7011 0.0908 1>}
  n3 = TextureSample() {texture=/VehicleCosmetics/Wheels/CamoSE/Textures/Tire_Camo_RGB}
  n4 = TextureSampleParameter2D() {group=00_Base, parameterName=Net Mask, sortPriority=0, texture=/VehicleCosmetics/Wheels/CamoSE/Textures/Wheel_CamoNet_SE_M}
  n5 = TextureSampleParameter2D() {
   group=00_Base
   parameterName=Net Normal
   samplerType=SAMPLERTYPE_Normal
   sortPriority=0
   texture=/VehicleCosmetics/Wheels/CamoSE/Textures/Wheel_CamoNet_SE_N
  }
  n6 = Transform(None=VectorParameter() {group=00_Base, parameterName=WindDirectionXYZ-RGB, defaultValue=<rgba 1 1 0 1>}.RGB) {transformSourceType=TRANSFORMSOURCE_Local}
  n7 = Rotator(Coordinate=TextureCoordinate(), Time=Constant() {r=-0.5}) {speed=3.14}
  n8 = ComponentMask(None=@n7) {r=true}
  n9 = Multiply(A=Time(), B=Multiply(A=ScalarParameter() {defaultValue=1.5, group=00_Base, parameterName=WindSpeed}))
  n10 = Sine(None=Add(A=OneMinus(None=@n8), B=@n9))
  n11 = Sine(None=Add(A=ComponentMask(None=@n7) {g=true}, B=@n9))
  n12 = Add(A=@n10, B=@n11)
  n13 = Add(A=@n10, B=@n12)
  n14 = Add(A=@n11, B=@n12)
  n15 = Multiply(A=Multiply(A=Multiply(A=@n13, B=@n14), B=VertexColor().A), B=ScalarParameter() {defaultValue=0.5, group=00_Base, parameterName=WindInt})
  n16 = SetMaterialAttributes!(
   MaterialAttributes
   EmissiveColor=MaterialFunctionCall(Base=LinearInterpolate(A=@n1.RGB, B=@n2.RGB, Alpha=@n3.G), Blend=@n4.G) {materialFunction=&Blend_Overlay}.Result
   Normal=@n5.RGB
   OpacityMask=@n4.R
   WorldPositionOffset=Multiply(A=Multiply(A=@n6, B=@n15), B=ScalarParameter() {defaultValue=1, group=00_Base, parameterName=animation})
   Roughness=Constant() {r=1}
   Metallic=Constant()
  ) {attributeSetTypes [
    #1
    #2
    #3
    #4
    #5
    #6
   ]}
  FunctionOutput(None=@n16) {outputName=Material Attributes}
 ]
}
