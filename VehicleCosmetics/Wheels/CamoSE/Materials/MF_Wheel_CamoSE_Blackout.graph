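# Material function for the Blackout variant of the CamoSE cosmetic wheel, producing a colored Fresnel rim glow as emissive.
# Rim and tire normal maps blend by vertex color red and feed the shared MF_Fresnel_Classic function borrowed from the Cola_Holographic wheel.
# The result is tinted by the Custom Color parameter, boosted, and output as Material Attributes with only EmissiveColor wired.
# A built-in Fresnel node is present but its output is unused.
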MaterialFunction /VehicleCosmetics/Wheels/CamoSE/Materials/MF_Wheel_CamoSE_Blackout
graph {nodes [
  n1 = ScalarParameter() {defaultValue=2, parameterName=Fresnel EXP}
  n2 = LinearInterpolate(
   A=TextureSampleParameter2D() {parameterName=Rim_N, samplerType=SAMPLERTYPE_Normal, texture=/VehicleCosmetics/Wheels/CamoSE/Textures/Wheel_Camo_SE_N}.RGBA
   B=TextureSampleParameter2D() {parameterName=Tier_N, samplerType=SAMPLERTYPE_Normal, texture=/VehicleCosmetics/Wheels/CamoSE/Textures/Tire_Camo_N}.RGBA
   Alpha=VertexColor().R
  )
  Fresnel(ExponentIn=@n1, BaseReflectFractionIn=ScalarParameter() {defaultValue=0.04, parameterName=BaseReflectFractionIn}, Normal=@n2) {exponent=2}
  n3 = Multiply(
   A=VectorParameter() {parameterName=Custom Color, defaultValue=<rgba 0 0.8796 0.1095 1>}.RGB
   B=MaterialFunctionCall!(Normal=@n2, Exponent=@n1) {materialFunction=/VehicleCosmetics/Wheels/Cola_Holographic/Materials/MF_Fresnel_Classic}.Result
  )
  FunctionOutput(None=SetMaterialAttributes!(MaterialAttributes, EmissiveColor=Multiply(A=@n3) {constB=1.5}) {attributeSetTypes [#1]}) {outputName=Material Attributes}
 ]}
