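# Unlit material for the Blackout variant of the CamoSE vehicle wheel cosmetic.
# The wheel's look comes from the MF_Wheel_CamoSE_Blackout material function, output as material attributes.
# That result passes through the shared MF_BaseVehicle_Damage function, then MF_VehicleDeRez_Transition, layering in vehicle damage and the de-rez transition.
# Usage is enabled for skeletal meshes and mesh deformers, with automatic usage setting in the editor turned off.
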Material /VehicleCosmetics/Wheels/CamoSE/Materials/M_Wheel_CamoSE_Blackout
graph {
 properties {
  bAutomaticallySetUsageInEditor=false
  bUseMaterialAttributes=true
  bUsedWithMeshDeformer=true
  bUsedWithSkeletalMesh=true
  shadingModel=MSM_Unlit
  usedShadingModels=Unlit
 }
 nodes [MaterialFunctionCall(In=MaterialFunctionCall(Material=MaterialFunctionCall() {materialFunction=&/MF_Wheel_CamoSE_Blackout}.Material Attributes) {materialFunction=&MF_BaseVehicle_Damage}.Result) {materialFunction=&MF_VehicleDeRez_Transition}]
}
